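; Intercom manager component that runs tag-keyed audio broadcast channels, heard through registered sound source buses.
; Once the game state is ready, it listens on the GameplayEventRouter for broadcast request and stop events.
; Requests start on free channels or interrupt non-blocking broadcasts of equal or lower priority; QueueBroadcast is an empty stub.
; Maps track listeners, active broadcasts and audio components; stops fade out, and removed listeners get their bus send zeroed.
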(event EventBeginPlay
  (|CallParentFunction)
  (Class|FortAsyncActionGameStateReady|GameStateReadyAsync
    (:OnReady
      (bind _returnvalue (GameplayEventRouter|GetEventRouter "/Script/FortniteGame.FortGameState" (Components|GetOwner)))
      (bind _output_get (Variables|Intercom|SetEvent%20Router _returnvalue))
      (Utilities|IsValid _output_get
        (:Is%20Valid
          (Utilities|FlowControl|Sequence
            (:then_0
              (bind (_asynctaskproxy _payload) (GameplayEventRouter|StartListeningToEvent (Variables|Intercom|GetEvent%20Router) "/IntercomAudio/Data/GameplayEvent_IntercomBroadcastRequest.GameplayEvent_IntercomBroadcastRequest")
                (:OnEventReceived
                  (Intercom|RequestBroadcast _payload))))
            (:then_1
              (Utilities|FlowControl|Sequence
                (:then_0
                  (bind (_asynctaskproxy_2 _payload_2) (GameplayEventRouter|StartListeningToEvent (Variables|Intercom|GetEvent%20Router) "/IntercomAudio/Data/GameplayEvent_IntercomStopBroadcast.GameplayEvent_IntercomStopBroadcast")
                    (:OnEventReceived
                      (Intercom|StopBroadcast _payload_2)
                      (Utilities|Map|Remove (Variables|Intercom|GetActive%20Channel%20Broadcasts) (Utilities|Struct|BreakGameplayEventIntercomStopBroadcast _payload_2))
                      (Intercom|CallOnBroadcastEnded (Utilities|Struct|BreakGameplayEventIntercomStopBroadcast _payload_2)))))))))))))

(fn AddListenertoChannel
  (params (Channel%20Tag :type Gameplay%20Tag%20Structure) (Source%20Bus :type Sound%20Source%20Bus%20Object%20Reference))
  (Intercom|RegisterSourceBus Channel%20Tag Source%20Bus)
  (Intercom|StartBroadcast (Variables|GetChannel%20Tag))
  (Intercom|CallOnListenerAdded (Variables|GetChannel%20Tag) (Variables|GetSource%20Bus)))

(fn QueueBroadcast
  (params (Channel%20Tag :type Gameplay%20Tag%20Structure) (Broadcast%20Request :type Gameplay%20Event%20Intercom%20Broadcast%20Request%20Structure)))

(fn RegisterSourceBus
  (params (Channel%20Tag :type Gameplay%20Tag%20Structure) (Source%20Bus :type Sound%20Source%20Bus%20Object%20Reference))
  (Utilities|IsValid Source%20Bus
    (:Is%20Valid
      (bind _output_get (Variables|Intercom|SetChannel%20Listeners (Utilities|Map|Find (Variables|Intercom|GetChannel%20Listeners) (Variables|GetChannel%20Tag))))
      (bind _output_get_2 (LocalVariables|SetSource%20Buses (Utilities|Struct|BreakIntercomChannelListeners _output_get)))
      (Utilities|Array|AddUnique _output_get_2 (Variables|GetSource%20Bus))
      (Utilities|Map|Add (Variables|Intercom|GetChannel%20Listeners) (Variables|GetChannel%20Tag) (Utilities|Struct|MakeIntercomChannelListeners (LocalVariables|GetSource%20Buses))))))

(fn RemoveListenerfromChannel
  (params (Channel%20Tag :type Gameplay%20Tag%20Structure) (Source%20Bus :type Sound%20Source%20Bus%20Object%20Reference))
  (bind _channel_tag (Variables|GetChannel%20Tag))
  (bind _channel_audio_components (Variables|Intercom|GetChannel%20Audio%20Components))
  (bind _value (Utilities|Map|Find _channel_audio_components _channel_tag))
  (Intercom|UnregisterSourceBus Channel%20Tag Source%20Bus)
  (Utilities|IsValid _value
    (:Is%20Valid
      (Audio|Components|Audio|SetSourceBusSendPostEffect _value (Variables|GetSource%20Bus) 0.0 :!class /Script/Engine.AudioComponent)
      (Intercom|CheckForListeners (Variables|GetChannel%20Tag) (Variables|GetSource%20Bus))
      (Intercom|CallOnListenerRemoved (Variables|GetChannel%20Tag) (Variables|GetSource%20Bus)))))

(fn RequestBroadcast
  (params (Broadcast%20Request%20Data_ChannelTag_2_7CBB941D4D8B12305FB174A284D2BE0C :type Gameplay%20Tag%20Structure) (Broadcast%20Request%20Data_Sound_5_E6C156894D5A0E2986A423BDC3A9A16C :type Sound%20Base%20Object%20Reference) (Broadcast%20Request%20Data_Priority_8_9D44E8734E35A1B467EFD28590B1B0B9 :type Integer) (Broadcast%20Request%20Data_IsBlocking_15_691F1A6045C7C88EA3C3F7AFD0EA541F :type Boolean) (Broadcast%20Request%20Data_QueueifOccupied_17_4342B4C348EAAE8774A2088805EC7A55 :type Boolean))
  (bind _channel_tag (LocalVariables|GetChannel%20Tag))
  (bind _active_channel_broadcasts (Variables|Intercom|GetActive%20Channel%20Broadcasts))
  (bind _value (Utilities|Map|Find _active_channel_broadcasts _channel_tag))
  (bind (_priority_8_9d44e8734e35a1b467efd28590b1b0b9 _isblocking_15_691f1a6045c7c88ea3c3f7afd0ea541f) (Utilities|Struct|BreakGameplayEventIntercomBroadcastRequest _value))
  (LocalVariables|SetChannel%20Tag Broadcast%20Request%20Data_ChannelTag_2_7CBB941D4D8B12305FB174A284D2BE0C)
  (if (not (Utilities|Map|Contains _active_channel_broadcasts _channel_tag))
    (Utilities|Map|Add (Variables|Intercom|GetActive%20Channel%20Broadcasts) (Utilities|Struct|BreakGameplayEventIntercomBroadcastRequest (Variables|GetBroadcast%20Request%20Data)) (Variables|GetBroadcast%20Request%20Data))
    (Intercom|StartBroadcast (LocalVariables|GetChannel%20Tag))
    (elif _isblocking_15_691f1a6045c7c88ea3c3f7afd0ea541f
      _
      (elif (> _priority_8_9d44e8734e35a1b467efd28590b1b0b9 (Utilities|Struct|BreakGameplayEventIntercomBroadcastRequest (Variables|GetBroadcast%20Request%20Data)))
        _
        (else
          (Intercom|InterruptBroadcast (LocalVariables|GetChannel%20Tag) (Variables|GetBroadcast%20Request%20Data))))))
  (if (Utilities|Struct|BreakGameplayEventIntercomBroadcastRequest (Variables|GetBroadcast%20Request%20Data))
    (Intercom|QueueBroadcast (LocalVariables|GetChannel%20Tag) (Variables|GetBroadcast%20Request%20Data))))

(fn StopBroadcast
  (params (Stop%20Broadcast%20Event%20Data_ChannelTag_2_7CBB941D4D8B12305FB174A284D2BE0C :type Gameplay%20Tag%20Structure) (Stop%20Broadcast%20Event%20Data_FadeOutDuration_19_4342B4C348EAAE8774A2088805EC7A55 :type Float%20%28double-precision%29))
  (LocalVariables|SetChannel%20Tag Stop%20Broadcast%20Event%20Data_ChannelTag_2_7CBB941D4D8B12305FB174A284D2BE0C)
  (Audio|Components|Audio|FadeOut (Utilities|Map|Find (Variables|Intercom|GetChannel%20Audio%20Components) (LocalVariables|GetChannel%20Tag)) Stop%20Broadcast%20Event%20Data_FadeOutDuration_19_4342B4C348EAAE8774A2088805EC7A55 0.0 "Linear" :!class /Script/Engine.AudioComponent)
  (Utilities|Map|Remove (Variables|Intercom|GetChannel%20Audio%20Components) (LocalVariables|GetChannel%20Tag)))

(fn UnregisterSourceBus
  (params (Channel%20Tag :type Gameplay%20Tag%20Structure) (Source%20Bus :type Sound%20Source%20Bus%20Object%20Reference))
  (bind _channel_tag (Variables|GetChannel%20Tag))
  (bind _channel_listeners (Variables|Intercom|GetChannel%20Listeners))
  (bind (_value _returnvalue) (Utilities|Map|Find _channel_listeners _channel_tag))
  (if _returnvalue
    (bind _output_get (LocalVariables|SetSource%20Buses (Utilities|Struct|BreakIntercomChannelListeners _value)))
    (Utilities|Array|RemoveIndex _output_get (Utilities|Array|FindItem _output_get (Variables|GetSource%20Bus)))
    (if (Utilities|Array|IsNotEmpty (LocalVariables|GetSource%20Buses))
      (Utilities|Map|Add (Variables|Intercom|GetChannel%20Listeners) (Variables|GetChannel%20Tag) (Utilities|Struct|MakeIntercomChannelListeners (LocalVariables|GetSource%20Buses)))
      (else
        (Utilities|Map|Remove (Variables|Intercom|GetChannel%20Listeners) (Variables|GetChannel%20Tag))))))
