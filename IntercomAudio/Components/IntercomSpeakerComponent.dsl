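; Actor component that turns its owner into an intercom speaker, playing a channel's audio through an attached source bus.
; Start Listening waits for playlist readiness, registers a lazily created source bus with the game state's IntercomManager, then spawns and fades in audio configured from a data asset (attenuation, effects, modulation, sends).
; Virtualization changes remove or re-add the channel listener.
; Stop Listening, also called on EndPlay, fades out, then deregisters.
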(event Custom|Start%20Listening
  (bind _speaker_audio_component (Variables|Intercom|GetSpeaker%20Audio%20Component))
  (Variables|Intercom|GetSpeaker%20Configuration%20Data
    (:then
      (bind _gamestate (Class|FortAsyncActionCurrentPlaylistReady|PlaylistDataReadyAsync)
        (:OnReady
          (bind _output_get (Variables|Intercom|SetIntercom%20Manager (Actor|GetComponentbyClass _gamestate "/IntercomAudio/Components/IntercomManager.IntercomManager_C")))
          (Utilities|IsValid _output_get
            (:Is%20Valid
              (Intercom|AddListenerToChannel (Variables|Intercom|GetIntercom%20Manager) (Variables|Intercom|GetChannel%20Tag) (Intercom|GetSourceBus))
              (bind _returnvalue (Audio|SpawnSoundAttached (Intercom|GetSourceBus) (Variables|/Script/Engine.Actor|GetRootComponent (Components|GetOwner)) "None" (Variables|/IntercomAudio/Data/IntercomSpeakerConfigurationData.IntercomSpeakerConfigurationData_C|GetLocation (Variables|Intercom|GetSpeaker%20Configuration%20Data)) (Variables|/IntercomAudio/Data/IntercomSpeakerConfigurationData.IntercomSpeakerConfigurationData_C|GetRotation (Variables|Intercom|GetSpeaker%20Configuration%20Data)) "KeepRelativeOffset" true (Variables|/IntercomAudio/Data/IntercomSpeakerConfigurationData.IntercomSpeakerConfigurationData_C|GetVolume%20Multiplier (Variables|Intercom|GetSpeaker%20Configuration%20Data)) (Variables|/IntercomAudio/Data/IntercomSpeakerConfigurationData.IntercomSpeakerConfigurationData_C|GetPitch%20Multiplier (Variables|Intercom|GetSpeaker%20Configuration%20Data)) 0.0 (Variables|/IntercomAudio/Data/IntercomSpeakerConfigurationData.IntercomSpeakerConfigurationData_C|GetAttenuation%20Settings (Variables|Intercom|GetSpeaker%20Configuration%20Data))))
              (bind _output_get_2 (Variables|Intercom|SetSpeaker%20Audio%20Component _returnvalue))
              (Utilities|IsValid _output_get_2
                (:Is%20Valid
                  (Utilities|FlowControl|Sequence
                    (:then_0
                      (Audio|Components|Audio|FadeIn (Variables|Intercom|GetSpeaker%20Audio%20Component) (Variables|/IntercomAudio/Data/IntercomSpeakerConfigurationData.IntercomSpeakerConfigurationData_C|GetFade%20In%20Duration (Variables|Intercom|GetSpeaker%20Configuration%20Data)) 1.0 0.0 "Linear" :!class /Script/Engine.AudioComponent))
                    (:then_1
                      (Utilities|FlowControl|Sequence
                        (:then_0
                          (EventDispatchers|BindEventtoOnAudioVirtualizationChanged (Variables|Intercom|GetSpeaker%20Audio%20Component) (EventDispatchers|CreateEvent :!event Custom|On%20Speaker%20Virtualization%20Changed)))
                        (:then_1
                          (Utilities|FlowControl|Sequence
                            (:then_0
                              (bind _source_effect_chain (Variables|/IntercomAudio/Data/IntercomSpeakerConfigurationData.IntercomSpeakerConfigurationData_C|GetSource%20Effect%20Chain (Variables|Intercom|GetSpeaker%20Configuration%20Data))
                                (:then
                                  (Variables|/Script/Engine.AudioComponent|SetSourceEffectChain _source_effect_chain (Variables|Intercom|GetSpeaker%20Audio%20Component)))
                                (:else)))
                            (:then_1
                              (Utilities|FlowControl|Sequence
                                (:then_0
                                  (Audio|Components|Audio|AddModulationRouting (Variables|Intercom|GetSpeaker%20Audio%20Component) (Variables|/IntercomAudio/Data/IntercomSpeakerConfigurationData.IntercomSpeakerConfigurationData_C|GetVolume%20Modulators (Variables|Intercom|GetSpeaker%20Configuration%20Data)) "Volume"))
                                (:then_1
                                  (Utilities|FlowControl|Sequence
                                    (:then_0
                                      (for _array_element (Variables|/IntercomAudio/Data/IntercomSpeakerConfigurationData.IntercomSpeakerConfigurationData_C|GetPre-Effect%20Bus%20Sends (Variables|Intercom|GetSpeaker%20Configuration%20Data))
                                        (Audio|Components|Audio|SetAudioBusSendPreEffect (Variables|Intercom|GetSpeaker%20Audio%20Component) _array_element 1.0 :!class /Script/Engine.AudioComponent)))
                                    (:then_1
                                      (Utilities|FlowControl|Sequence
                                        (:then_0
                                          (for _array_element_2 (Variables|/IntercomAudio/Data/IntercomSpeakerConfigurationData.IntercomSpeakerConfigurationData_C|GetPost-Effect%20Bus%20Sends (Variables|Intercom|GetSpeaker%20Configuration%20Data))
                                            (Audio|Components|Audio|SetAudioBusSendPostEffect (Variables|Intercom|GetSpeaker%20Audio%20Component) _array_element_2 1.0 :!class /Script/Engine.AudioComponent)))
                                        (:then_1
                                          (Utilities|FlowControl|Sequence
                                            (:then_0
                                              (bind _submix (Variables|/IntercomAudio/Data/IntercomSpeakerConfigurationData.IntercomSpeakerConfigurationData_C|GetSubmix (Variables|Intercom|GetSpeaker%20Configuration%20Data))
                                                (:then
                                                  (Audio|Components|Audio|SetSubmixSend _speaker_audio_component "/Game/Sounds/SubMixes/Fortnite_SFX.Fortnite_SFX" 0.0 :!class /Script/Engine.AudioComponent)
                                                  (Audio|Components|Audio|SetSubmixSend _speaker_audio_component _submix 1.0 :!class /Script/Engine.AudioComponent))
                                                (:else)))))))))))))))))))))))
    (:else
      ; Print Invalid Speaker Configuration Data
      (Development|PrintText (Utilities|Text|FormatText "NSLOCTEXT(\"[27FFB2BC34808C0461A01DA99D9E60EB]\", \"F65C0F4C41382D5C258CAD9DCD847E1B\", \"Unable to initialize Intercom Speaker Component for actor \\'{actor}\\'. Invalid Speaker Configuration Data.\")" (Components|GetOwner))))))

(event Custom|Stop%20Listening
  (bind _speaker_audio_component (Variables|Intercom|GetSpeaker%20Audio%20Component)
    (:then
      (Audio|Components|Audio|FadeOut _speaker_audio_component (Variables|/IntercomAudio/Data/IntercomSpeakerConfigurationData.IntercomSpeakerConfigurationData_C|GetFade%20Out%20Duration (Variables|Intercom|GetSpeaker%20Configuration%20Data)) 0.0 "Linear" :!class /Script/Engine.AudioComponent)
      (EventDispatchers|UnbindEventfromOnAudioVirtualizationChanged (Variables|Intercom|GetSpeaker%20Audio%20Component) (EventDispatchers|CreateEvent :!event Custom|On%20Speaker%20Virtualization%20Changed))
      (EventDispatchers|BindEventtoOnAudioFinished (Variables|Intercom|GetSpeaker%20Audio%20Component) (delegate Custom|Audio%20Finished)))
    (:else)))

(event Custom|Audio%20Finished
  (bind _intercom_manager (Variables|Intercom|GetIntercom%20Manager)
    (:then
      (Intercom|RemoveListenerFromChannel _intercom_manager (Variables|Intercom|GetChannel%20Tag) (Intercom|GetSourceBus)))
    (:else)))

(event Custom|On%20Speaker%20Virtualization%20Changed
  (params (bIsVirtualized :type Boolean))
  (if bIsVirtualized
    (bind _intercom_manager (Variables|Intercom|GetIntercom%20Manager)
      (:then
        (Intercom|RemoveListenerFromChannel _intercom_manager (Variables|Intercom|GetChannel%20Tag) (Intercom|GetSourceBus)))
      (:else))
    (else
      (bind _intercom_manager_2 (Variables|Intercom|GetIntercom%20Manager)
        (:then
          (Intercom|AddListenerToChannel _intercom_manager_2 (Variables|Intercom|GetChannel%20Tag) (Intercom|GetSourceBus)))
        (:else)))))

(event EventEndPlay
  (params (EndPlayReason :type EEndPlayReason%20Enum :default "Destroyed"))
  (|CallParentFunction EndPlayReason)
  (CallFunction|StopListening))

(fn GetSourceBus
  (bind _source_bus (Variables|Intercom|GetSource%20Bus)
    (:then
      (return _source_bus))
    (:else
      (bind _returnvalue (Game|ConstructObjectfromClass "/Script/Engine.SoundSourceBus"))
      (Variables|Intercom|SetSource%20Bus _returnvalue)
      (Variables|/Script/Engine.SoundBase|SetbBypassVolumeScaleForPriority (Variables|/IntercomAudio/Data/IntercomSpeakerConfigurationData.IntercomSpeakerConfigurationData_C|GetBypass%20Volume%20Scale%20for%20Priority (Variables|Intercom|GetSpeaker%20Configuration%20Data)) (Variables|Intercom|GetSource%20Bus))
      (Variables|/Script/Engine.SoundBase|SetPriority (Variables|/IntercomAudio/Data/IntercomSpeakerConfigurationData.IntercomSpeakerConfigurationData_C|GetPriority (Variables|Intercom|GetSpeaker%20Configuration%20Data)) (Variables|Intercom|GetSource%20Bus))
      (return (Variables|Intercom|GetSource%20Bus)))))
